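# Niagara cosmetic effect that puts pink glowing flame sprites on the hair of the female Glow Bro player character.
# Flame and HairDecal emitters pin partitioned sprites to socket-reader sockets with per-partition offsets, HairDecal using only the first two.
# Flame adds curve-driven color, scale and dynamic material parameters; HairDecal adds faint fading glow sprites.
# Both zero spawning and switch particle VisibilityTag when the head is hidden or System.NoGlow is set.
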NiagaraSystem /BRCosmetics/Characters/Player/Female/Medium/Bodies/F_MED_Glow_Bro/FX/NS_Female_GloBro_Hair
dependencies {
 usedDataInterfaces [
  &NiagaraDataInterfaceArrayFloat3
  &NiagaraDataInterfaceCurve
  &NiagaraDataInterfaceColorCurve
  &NiagaraDataInterfaceSocketReader
 ]
 usedDynamicInputs [&SetBoolByFloatComparison]
 usedModules [
  &SystemState
  &self:SystemSpawnScript.SystemScriptSource.SystemScriptGraph.NiagaraNodeAssignment_0.SetVariables_#1
  &self:SystemSpawnScript.SystemScriptSource.SystemScriptGraph.NiagaraNodeAssignment_1.SetVariables_#2
 ]
}
emitters {
 Flame {
  emitterUpdateScript [
   &EmitterState(Inactive Response=::NewEnumerator1, Life Cycle Mode=::NewEnumerator1, Loop Duration=1, UseLoopDelay=false)
   &SpawnRate(
    SpawnRate=&MakeCustomFloatFromBool(Bool=&Boolean_OrOperation(Bool1=$User.bisHeadHidden, Bool2=$System.NoGlow), True Float=0, False Float=4)
    Use Spawn Probability=false
   )
  ]
  particleSpawnScript [
   &PartitionParticles(Number Of Partitions=8)
   &InitializeParticle(
    Lifetime Max=1.5
    Lifetime Min=0.7
    Lifetime Mode=::NewEnumerator1
    Sprite Size Mode=::NewEnumerator1
    Uniform Sprite Size Max=50
    Uniform Sprite Size Min=25
    UsePositionOffset=false
    Write Material Random=true
   )
   &CameraOffset(Camera Offset Amount=4)
   &ShapeLocation(MaskSpawnGroup=false, Override Seed=false, Sphere Radius=0.5, UseOwnerScale=true)
   &AddVelocity(Velocity=&Add_Vector(
     A=&Multiply_VectorByFloat(Vector=$Particles.ShapeLocation.ShapeNormal, Float=1)
     B=&RandomRangeVector(Minimum=<0 3.6 2.5>, Maximum=<0 5.5 4.7>)
    ))
  ]
  particleUpdateScript [
   &ParticleState()
   &ScaleSpriteSize(Uniform Curve Sprite Scale=DataInterface() {propertyValues {Curve {keys [{tangentMode=RCTM_User, time=0, value=0.1}, <at 0.6 0.4>, {tangentMode=RCTM_User, time=1, value=0}]}}})
   &SocketLocation(
    SamplePosOffset=true
    Sampled Position Offset=&Add_Vector(
     A=Linked() {linkedVariable {name=Particles.Initial.ShapeLocation.ShapePosition, type=NiagaraPosition}}
     B=&SelectVectorFromArray(
      Direct Array Index=$Particles.PartitionParticles.Partition
      Vector Selection Array=DataInterface() {propertyValues {FloatData [<0 0 0>, <0 0 0>, <0 0 0>, <0 0 0>, <0 0 0>, <3 0 0>, <3 2 0>, <3 0 0>]}}
     )
    )
    Socket Index=$Particles.PartitionParticles.Partition
    Socket Reader=$System.Socket Reader
    Socket Sampling Mode=::NewEnumerator3
    VelocitySpeedLimit=false
   )
   &ScaleColor(
    Scale Alpha=&FloatFromCurve(FloatCurve=DataInterface() {propertyValues {Curve {keys [
         {arriveTangent=-5.8451, interpMode=RCIM_Cubic, leaveTangent=-5.8451, tangentMode=RCTM_User, time=0, value=1}
         <at 0.2 0.3 RCIM_Cubic>
         {arriveTangent=1.4672, interpMode=RCIM_Cubic, leaveTangent=1.4672, tangentMode=RCTM_User, time=1, value=1}
        ]}}})
    Scale RGB=&MakeVectorFromLinearColorRGB(Color=&ColorFromCurve(ColorCurve=DataInterface() {propertyValues {
        AlphaCurve {keys []}
        BlueCurve {keys [<at 0 0.7029 RCIM_Cubic>,<at 0.2 1.7548>,<at 1 0.3458 RCIM_Cubic>]}
        GreenCurve {keys [<at 0 0.2008 RCIM_Cubic>,<at 0.2 0.5227>,<at 1 0.0665 RCIM_Cubic>]}
        RedCurve {keys [<at 0 2.5017 RCIM_Cubic>,<at 0.2 4.0367>,<at 1 0.5 RCIM_Cubic>]}
       }}))
    ScaleA=true
    ScaleRGB=true
   )
   &DynamicMaterialParameters(
    Index 0 Param 1=&FloatFromCurve(FloatCurve=DataInterface() {propertyValues {Curve {keys [{arriveTangent=2.2577, interpMode=RCIM_Cubic, leaveTangent=2.2577, tangentMode=RCTM_User, time=0, value=0}, <at 1 1.5 RCIM_Cubic>]}}})
    Index 0 Param 2=&RandomRangeFloat(Minimum=0, Maximum=10)
    Index 0 Param 3=1
    Index 0 Param 4=1
    Module.Index0Float0Write=true
    Module.Index0Float1Write=true
    Module.Index0Float2Write=true
    Module.Index0Float3Write=false
   )
   &SolveForcesAndVelocity(Clamp Velocity=false, Limit Acceleration=false)
   /SetVariables(Particles.VisibilityTag=&MakeCustomIntFromBool(Bool=&Boolean_OrOperation(Bool1=$User.bisHeadHidden, Bool2=$System.NoGlow), True INT=1, False INT=0)) {isSetParameters=true, module=SetVariables_#3}
  ]
  rendererClasses [&NiagaraSpriteRendererProperties]
  renderers [{rendererClass=&NiagaraSpriteRendererProperties}]
 }
 HairDecal {
  emitterUpdateScript [
   &EmitterState()
   &SpawnRate(
    SpawnRate=&MakeCustomFloatFromBool(Bool=&Boolean_OrOperation(Bool1=$User.bisHeadHidden, Bool2=$System.NoGlow), True Float=0, False Float=16)
    Use Spawn Probability=false
   )
  ]
  particleSpawnScript [
   &PartitionParticles(Number Of Partitions=8)
   &InitializeParticle(
    Color=<rgba 1 0.1146 0.4844 0.05>
    Lifetime Min=1.5
    Lifetime Mode=::NewEnumerator1
    Sprite Rotation Angle=&RandomRangeFloat(Minimum=-0.05, Maximum=0.05)
    Sprite Rotation Mode=::NewEnumerator1
    Sprite Size Mode=::NewEnumerator1
    Uniform Sprite Size Max=12
    Uniform Sprite Size Min=10
    UsePositionOffset=false
    Write Material Random=true
   )
   &CameraOffset(Camera Offset Amount=8)
  ]
  particleUpdateScript [
   &ParticleState()
   &SocketLocation(
    SamplePosOffset=true
    Sampled Position Offset=&SelectVectorFromArray(
     Direct Array Index=$Particles.PartitionParticles.Partition
     Vector Selection Array=DataInterface() {propertyValues {FloatData [<0 0 0>, <0 0 0>, <0 0 0>, <0 0 0>, <0 0 0>, <3 0 0>, <3 2 0>, <3 0 0>]}}
    )
    Socket Index=&ModuloInt(A=$Particles.PartitionParticles.Partition, B=2)
    Socket Reader=$System.Socket Reader
    Socket Sampling Mode=::NewEnumerator3
    VelocitySpeedLimit=false
   )
   &ScaleColor(
    Scale Alpha=&FloatFromCurve(FloatCurve=DataInterface() {propertyValues {Curve {keys [
         {leaveTangentWeight=0.053, time=0, value=0}
         {arriveTangentWeight=0.0883, leaveTangentWeight=0.3283, time=0.159, value=1}
         {arriveTangentWeight=0.3061, interpMode=RCIM_Cubic, leaveTangentWeight=0.1295, tangentMode=RCTM_User, time=0.75, value=1}
         {arriveTangentWeight=0.0833, interpMode=RCIM_Cubic, tangentMode=RCTM_User, time=1, value=0}
        ]}}})
    Scale RGB=&VectorFromFloat(Value=&FloatFromCurve(FloatCurve=DataInterface() {propertyValues {Curve {keys [<at 0 3>,<at 1 1>]}}}))
    ScaleA=true
    ScaleRGB=true
   )
   /SetVariables(Particles.VisibilityTag=&MakeCustomIntFromBool(Bool=&Boolean_OrOperation(Bool1=$User.bisHeadHidden, Bool2=$System.NoGlow), True INT=1, False INT=0)) {isSetParameters=true, module=SetVariables_#3}
  ]
  rendererClasses [&NiagaraSpriteRendererProperties]
  renderers [{rendererClass=&NiagaraSpriteRendererProperties}]
 }
}
summary {
 systemName=NS_Female_GloBro_Hair
 emitters [
  {emitterName=HairDecal, rendererClasses [&NiagaraSpriteRendererProperties]}
  {emitterName=Flame, rendererClasses [&NiagaraSpriteRendererProperties]}
 ]
 userVariables [
  User.bisHandHidden {type=NiagaraBool}
  User.bisHeadHidden {type=NiagaraBool}
  User.NoGlowVariant {type=NiagaraFloat, defaultValue=0, description=""}
 ]
}
