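# Physics asset for secondary motion of the F_MED_Glow_Bro female body's jacket, covering collar, lapels, zipper, sleeves and hem panels.
# Small MSTR capsules root each dyn_ chain of one to three boxes or capsules, joining it to spine_03 with free rotation.
# Chain joints limit swing and tightly limit twist, while collar and lapels hinge on one swing axis.
# Pelvis, spine and thigh capsules supply body volumes.
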PhysicsAsset /BRCosmetics/Characters/Player/Female/Medium/Bodies/F_MED_Glow_Bro/Meshes/F_MED_Glow_Bro_Dyn
bodies {
 dyn_collar_bk_01 [{shapeType=Box, extentX=4.8125, extentY=7.5, extentZ=0.9824, center=<2.4 0 0>}]
 dyn_collar_bk_MSTR [{shapeType=Capsule, length=0.1, radius=0.1, rotation=<rot 90 0 0>}]
 dyn_collar_l_01 [{shapeType=Box, extentX=4.8125, extentY=7.5, extentZ=0.9824, center=<2.4 0 0>}]
 dyn_collar_l_MSTR [{shapeType=Capsule, length=0.1, radius=0.1, rotation=<rot 90 0 0>}]
 dyn_collar_r_01 [{shapeType=Box, extentX=4.8125, extentY=7.5, extentZ=0.9824, center=<-2.4 0 0>}]
 dyn_collar_r_MSTR [{shapeType=Capsule, length=0.1, radius=0.1, rotation=<rot 90 0 0>}]
 dyn_jacket_bk_01 [{shapeType=Box, extentX=7.5, extentY=7.5, extentZ=0.9824, center=<3.5872 0 0.0001>}]
 dyn_jacket_bk_MSTR [{shapeType=Capsule, length=0.1, radius=0.1, rotation=<rot 90 0 0>}]
 dyn_jacket_bk_l_01 [{shapeType=Box, extentX=7.5, extentY=7.5, extentZ=0.9824, center=<3.5872 0 0.0001>}]
 dyn_jacket_bk_l_MSTR [{shapeType=Capsule, length=0.1, radius=0.1, rotation=<rot 90 0 0>}]
 dyn_jacket_bk_r_01 [{shapeType=Box, extentX=7.5, extentY=7.5, extentZ=0.9824, center=<-3.5872 0 0>}]
 dyn_jacket_bk_r_MSTR [{shapeType=Capsule, length=0.1, radius=0.1, rotation=<rot 90 0 0>}]
 dyn_jacket_fr_l_01 [{shapeType=Box, extentX=7.5, extentY=7.5, extentZ=0.9824, center=<3.5872 0 0.0001>}]
 dyn_jacket_fr_l_02 [{shapeType=Box, extentX=7.5, extentY=5.0424, extentZ=0.9824, center=<3.5799 -0.6099 0.084>, rotation=<rot 4.9809 -0.4369 -5.0197>}]
 dyn_jacket_fr_l_03 [{shapeType=Box, extentX=7.5, extentY=3.8006, extentZ=0.9824, center=<3.5102 0.09 0.5077>, rotation=<rot 14.9999 0 10.0006>}]
 dyn_jacket_fr_l_MSTR [{shapeType=Capsule, length=0.1, radius=0.1, rotation=<rot 90 0 0>}]
 dyn_jacket_fr_r_01 [{shapeType=Box, extentX=7.5, extentY=7.5, extentZ=0.9824, center=<-3.5872 0 -0.0001>}]
 dyn_jacket_fr_r_02 [{shapeType=Box, extentX=7.5, extentY=5.0424, extentZ=0.9824, center=<-3.5799 0.0158 -0.136>, rotation=<rot 4.9809 -0.4369 -5.0197>}]
 dyn_jacket_fr_r_03 [{shapeType=Box, extentX=7.5, extentY=3.8006, extentZ=0.9824, center=<-3.3864 -1.0622 -0.2971>, rotation=<rot 13.0475 10.111 12.4582>}]
 dyn_jacket_fr_r_MSTR [{shapeType=Capsule, length=0.1, radius=0.1, rotation=<rot 90 0 0>}]
 dyn_jacket_l_01 [{shapeType=Box, extentX=7.5, extentY=7.5, extentZ=0.9824, center=<3.5872 0 0.0001>}]
 dyn_jacket_l_MSTR [{shapeType=Capsule, length=0.1, radius=0.1, rotation=<rot 90 0 0>}]
 dyn_jacket_r_01 [{shapeType=Box, extentX=7.5, extentY=7.5, extentZ=0.9824, center=<-3.5872 0 0>}]
 dyn_jacket_r_MSTR [{shapeType=Capsule, length=0.1, radius=0.1, rotation=<rot 90 0 0>}]
 dyn_lapel_fr_l_01 [{shapeType=Box, extentX=7.5, extentY=3.8909, extentZ=0.9824, center=<3.5 0 0>}]
 dyn_lapel_fr_l_MSTR [{shapeType=Capsule, length=0.1, radius=0.1, rotation=<rot 90 0 0>}]
 dyn_lapel_fr_r_01 [{shapeType=Box, extentX=4.499, extentY=5.5857, extentZ=0.9824, center=<1.6379 0 0.0003>}]
 dyn_lapel_fr_r_MSTR [{shapeType=Capsule, length=0.1, radius=0.1, rotation=<rot 90 0 0>}]
 dyn_sleeve_bk_l_01 [{shapeType=Capsule, length=3, radius=2, center=<2.2 0 0>, rotation=<rot 90 0 0>}]
 dyn_sleeve_bk_l_02 [{shapeType=Capsule, length=3, radius=2, center=<2.2 0 0>, rotation=<rot 90 0 0>}]
 dyn_sleeve_bk_l_MSTR [{shapeType=Capsule, length=0.1, radius=0.1, rotation=<rot 90 0 0>}]
 dyn_sleeve_bk_r_01 [{shapeType=Capsule, length=3, radius=2, center=<-2.2 0 0>, rotation=<rot 90 0 0>}]
 dyn_sleeve_bk_r_02 [{shapeType=Capsule, length=3, radius=2, center=<-2.2 0 0>, rotation=<rot 90 0 0>}]
 dyn_sleeve_bk_r_MSTR [{shapeType=Capsule, length=0.1, radius=0.1, rotation=<rot 90 0 0>}]
 dyn_sleeve_out_l_01 [{shapeType=Capsule, length=3, radius=2, center=<2.2 0 0>, rotation=<rot 90 0 0>}]
 dyn_sleeve_out_l_02 [{shapeType=Capsule, length=3, radius=2, center=<2.2 0 0>, rotation=<rot 90 0 0>}]
 dyn_sleeve_out_l_MSTR [{shapeType=Capsule, length=0.1, radius=0.1, rotation=<rot 90 0 0>}]
 dyn_sleeve_out_r_01 [{shapeType=Capsule, length=3, radius=2, center=<-2.2 0 0>, rotation=<rot 90 0 0>}]
 dyn_sleeve_out_r_02 [{shapeType=Capsule, length=3, radius=2, center=<-2.2 0 0>, rotation=<rot 90 0 0>}]
 dyn_sleeve_out_r_MSTR [{shapeType=Capsule, length=0.1, radius=0.1, rotation=<rot 90 0 0>}]
 dyn_zipper_l_01 [{shapeType=Box, extentX=3.2387, extentY=1.9237, extentZ=0.516, center=<0.9817 0 0.0001>}]
 dyn_zipper_l_MSTR [{shapeType=Capsule, length=0.1, radius=0.1, rotation=<rot 90 0 0>}]
 pelvis [{shapeType=Capsule, length=9.3457, radius=10.7641, center=<-2.154 -2.414 0.0185>}]
 spine_01 [{shapeType=Capsule, length=11.3319, radius=6.3125, center=<-0.1087 1.1129 -0.0036>}]
 spine_02 [{shapeType=Capsule, length=4.5693, radius=7.7722, center=<-0.7223 0.9036 -0.0012>}]
 spine_03 [{shapeType=Capsule, length=4.0793, radius=9.0295, center=<-1.0795 1.7555 0.0081>}]
 thigh_l [{shapeType=Capsule, length=17.1519, radius=8.6621, center=<-17.5209 0 0>, rotation=<rot 90 0 0>}]
 thigh_r [{shapeType=Capsule, length=17.1519, radius=8.6621, center=<17.5209 0 0>, rotation=<rot 90 0 0>}]
}
constraints [
 {bone1Name=dyn_sleeve_bk_l_01, bone2Name=dyn_sleeve_bk_l_MSTR, swing1LimitDegrees=45, swing2LimitDegrees=45, twistLimitDegrees=5}
 {bone1Name=dyn_sleeve_bk_l_02, bone2Name=dyn_sleeve_bk_l_01, swing1LimitDegrees=45, swing2LimitDegrees=45, twistLimitDegrees=5}
 {bone1Name=dyn_sleeve_out_l_01, bone2Name=dyn_sleeve_out_l_MSTR, swing1LimitDegrees=45, swing2LimitDegrees=45, twistLimitDegrees=5}
 {bone1Name=dyn_sleeve_out_l_02, bone2Name=dyn_sleeve_out_l_01, swing1LimitDegrees=45, swing2LimitDegrees=45, twistLimitDegrees=5}
 {bone1Name=dyn_sleeve_out_r_01, bone2Name=dyn_sleeve_out_r_MSTR, swing1LimitDegrees=45, swing2LimitDegrees=45, twistLimitDegrees=5}
 {bone1Name=dyn_sleeve_out_r_02, bone2Name=dyn_sleeve_out_r_01, swing1LimitDegrees=45, swing2LimitDegrees=45, twistLimitDegrees=5}
 {bone1Name=dyn_sleeve_bk_r_01, bone2Name=dyn_sleeve_bk_r_MSTR, swing1LimitDegrees=45, swing2LimitDegrees=45, twistLimitDegrees=5}
 {bone1Name=dyn_sleeve_bk_r_02, bone2Name=dyn_sleeve_bk_r_01, swing1LimitDegrees=45, swing2LimitDegrees=45, twistLimitDegrees=5}
 {
  bone1Name=dyn_collar_bk_01
  bone2Name=dyn_collar_bk_MSTR
  swing1LimitDegrees=45
  swing1Motion=Locked
  swing2LimitDegrees=60
  twistLimitDegrees=45
  twistMotion=Locked
 }
 {
  bone1Name=dyn_collar_l_01
  bone2Name=dyn_collar_l_MSTR
  swing1LimitDegrees=45
  swing1Motion=Locked
  swing2LimitDegrees=45
  twistLimitDegrees=45
  twistMotion=Locked
 }
 {
  bone1Name=dyn_collar_r_01
  bone2Name=dyn_collar_r_MSTR
  swing1LimitDegrees=45
  swing1Motion=Locked
  swing2LimitDegrees=45
  twistLimitDegrees=45
  twistMotion=Locked
 }
 {bone1Name=dyn_zipper_l_01, bone2Name=dyn_zipper_l_MSTR, swing1LimitDegrees=45, swing2LimitDegrees=45, twistLimitDegrees=5}
 {
  bone1Name=dyn_lapel_fr_l_01
  bone2Name=dyn_lapel_fr_l_MSTR
  swing1LimitDegrees=45
  swing1Motion=Locked
  swing2LimitDegrees=45
  twistLimitDegrees=45
  twistMotion=Locked
 }
 {
  bone1Name=dyn_lapel_fr_r_01
  bone2Name=dyn_lapel_fr_r_MSTR
  swing1LimitDegrees=45
  swing1Motion=Locked
  swing2LimitDegrees=45
  twistLimitDegrees=45
  twistMotion=Locked
 }
 {bone1Name=dyn_jacket_fr_l_01, bone2Name=dyn_jacket_fr_l_MSTR, swing1LimitDegrees=45, swing2LimitDegrees=45, twistLimitDegrees=3}
 {bone1Name=dyn_jacket_fr_l_02, bone2Name=dyn_jacket_fr_l_01, swing1LimitDegrees=45, swing2LimitDegrees=45, twistLimitDegrees=3}
 {bone1Name=dyn_jacket_fr_l_03, bone2Name=dyn_jacket_fr_l_02, swing1LimitDegrees=45, swing2LimitDegrees=45, twistLimitDegrees=3}
 {bone1Name=dyn_jacket_l_01, bone2Name=dyn_jacket_l_MSTR, swing1LimitDegrees=45, swing2LimitDegrees=45, twistLimitDegrees=3}
 {bone1Name=dyn_jacket_bk_l_01, bone2Name=dyn_jacket_bk_l_MSTR, swing1LimitDegrees=45, swing2LimitDegrees=45, twistLimitDegrees=3}
 {bone1Name=dyn_jacket_bk_01, bone2Name=dyn_jacket_bk_MSTR, swing1LimitDegrees=45, swing2LimitDegrees=45, twistLimitDegrees=3}
 {bone1Name=dyn_jacket_bk_r_01, bone2Name=dyn_jacket_bk_r_MSTR, swing1LimitDegrees=45, swing2LimitDegrees=45, twistLimitDegrees=3}
 {bone1Name=dyn_jacket_r_01, bone2Name=dyn_jacket_r_MSTR, swing1LimitDegrees=45, swing2LimitDegrees=45, twistLimitDegrees=3}
 {bone1Name=dyn_jacket_fr_r_01, bone2Name=dyn_jacket_fr_r_MSTR, swing1LimitDegrees=45, swing2LimitDegrees=45, twistLimitDegrees=3}
 {bone1Name=dyn_jacket_fr_r_02, bone2Name=dyn_jacket_fr_r_01, swing1LimitDegrees=45, swing2LimitDegrees=45, twistLimitDegrees=3}
 {bone1Name=dyn_jacket_fr_r_03, bone2Name=dyn_jacket_fr_r_02, swing1LimitDegrees=45, swing2LimitDegrees=45, twistLimitDegrees=3}
 {
  bone1Name=dyn_sleeve_bk_l_MSTR
  bone2Name=spine_03
  swing1LimitDegrees=45
  swing1Motion=Free
  swing2LimitDegrees=45
  swing2Motion=Free
  twistLimitDegrees=45
  twistMotion=Free
 }
 {
  bone1Name=dyn_sleeve_out_l_MSTR
  bone2Name=spine_03
  swing1LimitDegrees=45
  swing1Motion=Free
  swing2LimitDegrees=45
  swing2Motion=Free
  twistLimitDegrees=45
  twistMotion=Free
 }
 {
  bone1Name=dyn_sleeve_out_r_MSTR
  bone2Name=spine_03
  swing1LimitDegrees=45
  swing1Motion=Free
  swing2LimitDegrees=45
  swing2Motion=Free
  twistLimitDegrees=45
  twistMotion=Free
 }
 {
  bone1Name=dyn_sleeve_bk_r_MSTR
  bone2Name=spine_03
  swing1LimitDegrees=45
  swing1Motion=Free
  swing2LimitDegrees=45
  swing2Motion=Free
  twistLimitDegrees=45
  twistMotion=Free
 }
 {
  bone1Name=dyn_collar_bk_MSTR
  bone2Name=spine_03
  swing1LimitDegrees=45
  swing1Motion=Free
  swing2LimitDegrees=45
  swing2Motion=Free
  twistLimitDegrees=45
  twistMotion=Free
 }
 {
  bone1Name=dyn_collar_l_MSTR
  bone2Name=spine_03
  swing1LimitDegrees=45
  swing1Motion=Free
  swing2LimitDegrees=45
  swing2Motion=Free
  twistLimitDegrees=45
  twistMotion=Free
 }
 {
  bone1Name=dyn_collar_r_MSTR
  bone2Name=spine_03
  swing1LimitDegrees=45
  swing1Motion=Free
  swing2LimitDegrees=45
  swing2Motion=Free
  twistLimitDegrees=45
  twistMotion=Free
 }
 {
  bone1Name=dyn_zipper_l_MSTR
  bone2Name=spine_03
  swing1LimitDegrees=45
  swing1Motion=Free
  swing2LimitDegrees=45
  swing2Motion=Free
  twistLimitDegrees=45
  twistMotion=Free
 }
 {
  bone1Name=dyn_lapel_fr_l_MSTR
  bone2Name=spine_03
  swing1LimitDegrees=45
  swing1Motion=Free
  swing2LimitDegrees=45
  swing2Motion=Free
  twistLimitDegrees=45
  twistMotion=Free
 }
 {
  bone1Name=dyn_lapel_fr_r_MSTR
  bone2Name=spine_03
  swing1LimitDegrees=45
  swing1Motion=Free
  swing2LimitDegrees=45
  swing2Motion=Free
  twistLimitDegrees=45
  twistMotion=Free
 }
 {
  bone1Name=dyn_jacket_fr_l_MSTR
  bone2Name=spine_03
  swing1LimitDegrees=45
  swing1Motion=Free
  swing2LimitDegrees=45
  swing2Motion=Free
  twistLimitDegrees=45
  twistMotion=Free
 }
 {
  bone1Name=dyn_jacket_l_MSTR
  bone2Name=spine_03
  swing1LimitDegrees=45
  swing1Motion=Free
  swing2LimitDegrees=45
  swing2Motion=Free
  twistLimitDegrees=45
  twistMotion=Free
 }
 {
  bone1Name=dyn_jacket_bk_l_MSTR
  bone2Name=spine_03
  swing1LimitDegrees=45
  swing1Motion=Free
  swing2LimitDegrees=45
  swing2Motion=Free
  twistLimitDegrees=45
  twistMotion=Free
 }
 {
  bone1Name=dyn_jacket_bk_MSTR
  bone2Name=spine_03
  swing1LimitDegrees=45
  swing1Motion=Free
  swing2LimitDegrees=45
  swing2Motion=Free
  twistLimitDegrees=45
  twistMotion=Free
 }
 {
  bone1Name=dyn_jacket_bk_r_MSTR
  bone2Name=spine_03
  swing1LimitDegrees=45
  swing1Motion=Free
  swing2LimitDegrees=45
  swing2Motion=Free
  twistLimitDegrees=45
  twistMotion=Free
 }
 {
  bone1Name=dyn_jacket_r_MSTR
  bone2Name=spine_03
  swing1LimitDegrees=45
  swing1Motion=Free
  swing2LimitDegrees=45
  swing2Motion=Free
  twistLimitDegrees=45
  twistMotion=Free
 }
 {
  bone1Name=dyn_jacket_fr_r_MSTR
  bone2Name=spine_03
  swing1LimitDegrees=45
  swing1Motion=Free
  swing2LimitDegrees=45
  swing2Motion=Free
  twistLimitDegrees=45
  twistMotion=Free
 }
 {
  bone1Name=spine_03
  bone2Name=spine_02
  swing1LimitDegrees=45
  swing1Motion=Free
  swing2LimitDegrees=45
  swing2Motion=Free
  twistLimitDegrees=45
  twistMotion=Free
 }
 {
  bone1Name=spine_02
  bone2Name=spine_01
  swing1LimitDegrees=45
  swing1Motion=Free
  swing2LimitDegrees=45
  swing2Motion=Free
  twistLimitDegrees=45
  twistMotion=Free
 }
 {
  bone1Name=spine_01
  bone2Name=pelvis
  swing1LimitDegrees=45
  swing1Motion=Free
  swing2LimitDegrees=45
  swing2Motion=Free
  twistLimitDegrees=45
  twistMotion=Free
 }
 {
  bone1Name=thigh_l
  bone2Name=pelvis
  swing1LimitDegrees=45
  swing1Motion=Free
  swing2LimitDegrees=45
  swing2Motion=Free
  twistLimitDegrees=45
  twistMotion=Free
 }
 {
  bone1Name=thigh_r
  bone2Name=pelvis
  swing1LimitDegrees=45
  swing1Motion=Free
  swing2LimitDegrees=45
  swing2Motion=Free
  twistLimitDegrees=45
  twistMotion=Free
 }
]
