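# Post-process style Anim Blueprint for the F_MED_Glow_Bro female cosmetic body, adding secondary motion to its sleeves, lapels and collar.
# It copies the attached parent mesh's pose, then drives dyn_*_MSTR bones with weighted multi-bone constraints.
# A RigidBody node then simulates the dedicated physics asset in spine_02 space, fed by a skydiving external force.
# EventGraph logic computes arm-raise values from upper-arm yaw, but the AnimGraph never reads them.
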AnimBlueprint /BRCosmetics/Characters/Player/Female/Medium/Bodies/F_MED_Glow_Bro/Meshes/F_MED_Glow_Bro_AnimBP
target_skeleton=&/F_MED_Glow_Bro_Skeleton
graphs [
 AnimGraph {nodes [
   n1 = "Constraint - dyn_collar_l_MSTR (P)"(Alpha=1, ConstraintWeights_0=1) {
    alphaBoolBlend {blendOption=Linear}
    boneToModify {boneName=dyn_collar_l_MSTR}
    constraintSetup [{targetBone {boneName=clavicle_l}}]
    constraintWeights [1]
   }
   "Constraint - dyn_collar_r_MSTR (P)"(ComponentPose=@n1, Alpha=1, ConstraintWeights_0=1) {
    alphaBoolBlend {blendOption=Linear}
    boneToModify {boneName=dyn_collar_r_MSTR}
    constraintSetup [{targetBone {boneName=clavicle_r}}]
    constraintWeights [1]
   }
   n2 = "Constraint - dyn_sleeve_bk_l_MSTR (PPPP)"!(
    ComponentPose=Local To Component(LocalPose=Copy Pose From Mesh() {bUseAttachedParent=true})
    Alpha=1
    ConstraintWeights_0=0.1137
    ConstraintWeights_1:float=0.1067
    ConstraintWeights_2:float=0.6398
    ConstraintWeights_3:float=0.1398
   ) -> (Pose) {
    alphaBoolBlend {blendOption=Linear}
    boneToModify {boneName=dyn_sleeve_bk_l_MSTR}
    constraintSetup [
     {targetBone {boneName=upperarm_twist_02_l}}
     {targetBone {boneName=upperarm_l}}
     {targetBone {boneName=upperarm_twist_01_l}}
     {targetBone {boneName=dyn_sleeve_bk_l_01}}
    ]
    constraintWeights [1,1,1,1]
   }
   n3 = "Constraint - dyn_sleeve_out_l_MSTR (PPPP)"!(
    ComponentPose=@n2
    Alpha=1
    ConstraintWeights_0=0.2005
    ConstraintWeights_1:float=0.1739
    ConstraintWeights_2:float=0.5097
    ConstraintWeights_3:float=0.116
   ) -> (Pose) {
    alphaBoolBlend {blendOption=Linear}
    boneToModify {boneName=dyn_sleeve_out_l_MSTR}
    constraintSetup [
     {targetBone {boneName=upperarm_twist_02_l}}
     {targetBone {boneName=upperarm_l}}
     {targetBone {boneName=upperarm_twist_01_l}}
     {targetBone {boneName=dyn_sleeve_out_l_01}}
    ]
    constraintWeights [1,1,1,1]
   }
   n4 = "Constraint - dyn_sleeve_out_r_MSTR (PPPP)"!(
    ComponentPose=@n3
    Alpha=1
    ConstraintWeights_0=0.5097
    ConstraintWeights_1:float=0.1739
    ConstraintWeights_2:float=0.2005
    ConstraintWeights_3:float=0.116
   ) -> (Pose) {
    alphaBoolBlend {blendOption=Linear}
    boneToModify {boneName=dyn_sleeve_out_r_MSTR}
    constraintSetup [
     {targetBone {boneName=upperarm_twist_01_r}}
     {targetBone {boneName=upperarm_r}}
     {targetBone {boneName=upperarm_twist_02_r}}
     {targetBone {boneName=dyn_sleeve_out_r_01}}
    ]
    constraintWeights [1,1,1,1]
   }
   n5 = "Constraint - dyn_sleeve_bk_r_MSTR (PPPP)"!(
    ComponentPose=@n4
    Alpha=1
    ConstraintWeights_0=0.6398
    ConstraintWeights_1:float=0.1398
    ConstraintWeights_2:float=0.1067
    ConstraintWeights_3:float=0.1137
   ) -> (Pose) {
    alphaBoolBlend {blendOption=Linear}
    boneToModify {boneName=dyn_sleeve_bk_r_MSTR}
    constraintSetup [
     {targetBone {boneName=upperarm_twist_01_r}}
     {targetBone {boneName=dyn_sleeve_bk_r_01}}
     {targetBone {boneName=upperarm_r}}
     {targetBone {boneName=upperarm_twist_02_r}}
    ]
    constraintWeights [1,1,1,1]
   }
   n6 = "Constraint - dyn_lapel_fr_l_MSTR (PPP)"!(ComponentPose=@n5, Alpha=1, ConstraintWeights_0=0.6894, ConstraintWeights_1:float=0.0792, ConstraintWeights_2:float=0.2314) -> (Pose) {
    alphaBoolBlend {blendOption=Linear}
    boneToModify {boneName=dyn_lapel_fr_l_MSTR}
    constraintSetup [{targetBone {boneName=spine_04}}, {targetBone {boneName=spine_05}}, {targetBone {boneName=dyn_lapel_fr_l_01}}]
    constraintWeights [1,1,1]
   }
   n7 = "Constraint - dyn_lapel_fr_r_MSTR (PP)"!(ComponentPose=@n6, Alpha=1, ConstraintWeights_0=0.9606, ConstraintWeights_1:float=0.0394) -> (Pose) {
    alphaBoolBlend {blendOption=Linear}
    boneToModify {boneName=dyn_lapel_fr_r_MSTR}
    constraintSetup [{targetBone {boneName=spine_04}}, {targetBone {boneName=spine_05}}]
    constraintWeights [1,1]
   }
   n8 = "Constraint - dyn_collar_l_MSTR (PP)"!(ComponentPose=@n7, Alpha=1, ConstraintWeights_0=0.7, ConstraintWeights_1:float=0.3) -> (Pose) {
    alphaBoolBlend {blendOption=Linear}
    boneToModify {boneName=dyn_collar_l_MSTR}
    constraintSetup [{targetBone {boneName=clavicle_l}}, {targetBone {boneName=spine_05}}]
    constraintWeights [1,1]
   }
   n9 = "Constraint - dyn_collar_r_MSTR (PP)"!(ComponentPose=@n8, Alpha=1, ConstraintWeights_0=0.3, ConstraintWeights_1:float=0.7) -> (Pose) {
    alphaBoolBlend {blendOption=Linear}
    boneToModify {boneName=dyn_collar_r_MSTR}
    constraintSetup [{targetBone {boneName=spine_05}}, {targetBone {boneName=clavicle_r}}]
    constraintWeights [1,1]
   }
   n10 = RigidBody(
    ComponentPose=@n9
    ExternalForce=Break Fort Anim Input Skydiving External Force(FortAnimInput_SkydivingExternalForce=Get SkydivingExternalForceInput()) -> (
     bUseSkydivingVectorForce=~
     SkydivingVectorForce=~
     HeadToPelvisDirection=~
     FloatingMultiplier=~
     FloatingAdditive=~
     DivingMultiplier=~
     DivingAdditive=~
     ParachutingMultiplier=~
     ParachutingAdditive=~
    ).SkydivingVectorForce
   ) {
    bClampLinearTranslationLimitToRefPose=true
    bOverrideWorldGravity=true
    overridePhysicsAsset=&/F_MED_Glow_Bro_Dyn
    simulationSpace=BaseBoneSpace
    alphaBoolBlend {blendOption=Linear}
    baseBoneRef {boneName=spine_02}
    componentAppliedLinearAccClamp=<3000 3000 3000>
    componentLinearAccScale=<0.3 0.2 0.2>
    componentLinearVelScale=<0.3 0.2 0.2>
    overrideWorldGravity {z=-980}
   }
   Output Pose(Result=Component To Local(ComponentPose=@n10)) {layerGroup=None}
  ]}
 EventGraph {nodes [
   Reroute Node()
   n1 = GetOwnerSkeletalMeshComponent()
   n2 = BreakRotator(InRot=BreakTransform(InTransform=GetDeltaTransformFromRefPose(self=@n1, BoneName=upperarm_r, BaseName=spine_05)).Rotation)
   > Set arm_up_r(arm_up_r:double=MapRangeClamped(Value="Absolute (Float)"(A=@n2.Yaw), InRangeA=60, InRangeB=100)) -> (Output_Get:double)
   n3 = GetOwnerSkeletalMeshComponent()
   n4 = BreakRotator(InRot=BreakTransform(InTransform=GetDeltaTransformFromRefPose(self=@n3, BoneName=upperarm_l, BaseName=spine_05)).Rotation)
   > Set arm_up_l(arm_up_l:double=MapRangeClamped(Value="Absolute (Float)"(A=@n4.Yaw), InRangeA=60, InRangeB=100)) -> (Output_Get:double)
   Event BlueprintUpdateAnimation()
  ]}
]
variables [
 arm_up_l {type=number}
 arm_up_r {type=number}
]
