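# Physics asset for the face accessory mesh of the F_MED_Glow_Bro player cosmetic, covering its dynamic hair and hat bones.
# Eight identical box bodies sit on front, left, right and back hair bones, with capsules on the hat bone and head.
# Each hair and hat body is constrained directly to the head, not chained, with the same moderate swing limits and tight twist limit.
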PhysicsAsset /BRCosmetics/Characters/Player/Female/Medium/Bodies/F_MED_Glow_Bro/Meshes/Parts/F_MED_Glow_Bro_FaceAcc_Dyn
bodies {
 dyn_hair_bk_06 [{shapeType=Box, extentX=5.8125, extentY=4.0625, extentZ=0.875, center=<2.5 0 0>}]
 dyn_hair_bk_07 [{shapeType=Box, extentX=5.8125, extentY=4.0625, extentZ=0.875, center=<2.5 0 0>}]
 dyn_hair_bk_08 [{shapeType=Box, extentX=5.8125, extentY=4.0625, extentZ=0.875, center=<2.5 0 0>}]
 dyn_hair_fr_01 [{shapeType=Box, extentX=5.8125, extentY=4.0625, extentZ=0.875, center=<2.5 0 0>}]
 dyn_hair_l_02 [{shapeType=Box, extentX=5.8125, extentY=4.0625, extentZ=0.875, center=<2.5 0 0>}]
 dyn_hair_r_03 [{shapeType=Box, extentX=5.8125, extentY=4.0625, extentZ=0.875, center=<2.5 0 0>}]
 dyn_hair_r_04 [{shapeType=Box, extentX=5.8125, extentY=4.0625, extentZ=0.875, center=<2.5 0 0>}]
 dyn_hair_r_05 [{shapeType=Box, extentX=5.8125, extentY=4.0625, extentZ=0.875, center=<2.5 0 0>}]
 dyn_hat_01 [{shapeType=Capsule, length=2, radius=2, center=<0.2 0 0>, rotation=<rot 90 0 0>}]
 head [{shapeType=Capsule, length=2, radius=2, rotation=<rot 90 0 0>}]
}
constraints [
 {bone1Name=dyn_hat_01, bone2Name=head, swing1LimitDegrees=45, swing2LimitDegrees=45, twistLimitDegrees=5}
 {bone1Name=dyn_hair_fr_01, bone2Name=head, swing1LimitDegrees=45, swing2LimitDegrees=45, twistLimitDegrees=5}
 {bone1Name=dyn_hair_l_02, bone2Name=head, swing1LimitDegrees=45, swing2LimitDegrees=45, twistLimitDegrees=5}
 {bone1Name=dyn_hair_r_03, bone2Name=head, swing1LimitDegrees=45, swing2LimitDegrees=45, twistLimitDegrees=5}
 {bone1Name=dyn_hair_r_04, bone2Name=head, swing1LimitDegrees=45, swing2LimitDegrees=45, twistLimitDegrees=5}
 {bone1Name=dyn_hair_r_05, bone2Name=head, swing1LimitDegrees=45, swing2LimitDegrees=45, twistLimitDegrees=5}
 {bone1Name=dyn_hair_bk_06, bone2Name=head, swing1LimitDegrees=45, swing2LimitDegrees=45, twistLimitDegrees=5}
 {bone1Name=dyn_hair_bk_07, bone2Name=head, swing1LimitDegrees=45, swing2LimitDegrees=45, twistLimitDegrees=5}
 {bone1Name=dyn_hair_bk_08, bone2Name=head, swing1LimitDegrees=45, swing2LimitDegrees=45, twistLimitDegrees=5}
]
